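# Animation Blueprint for the WWII Pilot Sci-Fi face accessory, copying the attached parent's pose and adjusting the dyn_knife bone.
# Additive rotations offset the bone when crouching and further when crouched and moving.
# Each update snaps both blend alphas to 0 or 1 from bIsCrouching and PawnSpeed2D.
# A heavily damped, angle-limited AnimDynamics node then adds partial-strength secondary motion relative to root.
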AnimBlueprint /BRCosmetics/Accessories/Hats/M_MED_WWII_Pilot_Sci_Fi_FaceAcc/Meshes/M_MED_WWII_Pilot_Sci_Fi_FaceAcc_AnimBp
target_skeleton=&/M_MED_WWII_Pilot_Sci_Fi_FaceAcc_Skeleton
graphs [
 AnimGraph {nodes [
   n1 = "Transform (Modify) Bone - Bone: dyn_knife"(
    ComponentPose=Local To Component(LocalPose=Copy Pose From Mesh() {bUseAttachedParent=true})
    Alpha=Get CrouchSprint_Alpha() -> (CrouchSprint_Alpha:double)
    Translation="0,0,0"
    Rotation="0,10,-10"
    Scale="1,1,1"
   ) {
    rotationMode=BMM_Additive
    rotationSpace=BCS_ParentBoneSpace
    alphaBoolBlend {blendOption=Linear}
    boneToModify {boneName=dyn_knife}
    rotation {roll=-10, yaw=10}
   }
   n2 = "Transform (Modify) Bone - Bone: dyn_knife"(ComponentPose=@n1, Alpha=Get Crouch_Alpha() -> (Crouch_Alpha:double), Translation="0,0,0", Rotation="0,60,0", Scale="1,1,1") {
    rotationMode=BMM_Additive
    rotationSpace=BCS_ParentBoneSpace
    alphaBoolBlend {blendOption=Linear}
    boneToModify {boneName=dyn_knife}
    rotation {yaw=60}
   }
   n3 = Anim Dynamics - Bone: dyn_knife(ComponentPose=@n2, Alpha=0.75, ExternalForce="0,0,0") {
    angularDampingOverride=0.9999
    bGravityOverrideInSimSpace=true
    bOverrideAngularDamping=true
    bOverrideLinearDamping=true
    linearDampingOverride=0.9999
    alphaBoolBlend {blendOption=Linear}
    boundBone {boneName=dyn_knife}
    componentAppliedLinearAccClamp=<10000 10000 10000>
    componentLinearAccScale=<0.5 0.5 0.5>
    componentLinearVelScale=<0.5 0.5 0.5>
    physicsBodyDefinitions [{
      sphereCollisionRadius=0
      boundBone {boneName=dyn_knife}
      boxExtents=<15 7 2>
      constraintSetup {angularLimitsMax=<0 0 20>, angularLimitsMin=<0 0 -20>, angularTarget=<0 0 0>}
      localJointOffset=<6 0 0>
     }]
    relativeSpaceBone {boneName=root}
   }
   Output Pose(Result=Component To Local(ComponentPose=@n3)) {layerGroup=None}
  ]}
 EventGraph {nodes [
   Event BlueprintUpdateAnimation()
   > Set Crouch_Alpha(Crouch_Alpha:double=SelectFloat(A=1, B=0, bPickA=Get bIsCrouching())) -> (Output_Get:double)
   n1 = "float > float"(A=Get PawnSpeed2D(), B=0) {is=title K2Node_CallFunction}
   n2 = Get bIsCrouching()
   > Set CrouchSprint_Alpha(CrouchSprint_Alpha:double=SelectFloat(A=1, B=0, bPickA=AND Boolean(A=@n1, B=@n2))) -> (Output_Get:double)
   TryGetPawnOwner()
  ]}
]
variables [
 Crouch_Alpha {type=number}
 CrouchSprint_Alpha {type=number}
]
